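# Physics asset for the M_MED_OatmealSpreadActive male medium player body cosmetic, covering the lower body and two dangling short-string chains.
# Capsules on the pelvis, thighs and calves form the leg hierarchy, joined by constraints with uniform swing and twist limits.
# Left and right four-capsule dyn_shortString chains hang from the pelvis, with twist limits set to zero and a tighter secondary swing limit at each second link.
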PhysicsAsset /BRCosmetics/Characters/Player/Male/Medium/Bodies/M_MED_OatmealSpreadActive/Meshes/M_MED_OatmealSpreadActive_Physics
bodies {
 calf_l [{shapeType=Capsule, length=34.5, radius=7, center=<-17.7026 -1.9577 -0.5552>, rotation=<rot -90 0 -0.0001>}]
 calf_r [{shapeType=Capsule, length=34.5, radius=7, center=<17.7026 1.5842 0.8815>, rotation=<rot 90 179.9204 179.9217>}]
 dyn_shortString_l_01 [{shapeType=Capsule, length=2, radius=1.5, shapeName=dyn_shortString_l_01_capsule, center=<1.5373 0 0>, rotation=<rot -90 -3.0357 3.0357>}]
 dyn_shortString_l_02 [{shapeType=Capsule, length=2, radius=1.5, shapeName=dyn_shortString_l_02_capsule, center=<1.8175 0 0>, rotation=<rot 90 180 -180>}]
 dyn_shortString_l_03 [{shapeType=Capsule, length=3, radius=1.5, shapeName=dyn_shortString_l_03_capsule, center=<1.8175 0 0>, rotation=<rot 90 -179.9604 -179.9605>}]
 dyn_shortString_l_04 [{shapeType=Capsule, length=3, radius=1.5, shapeName=dyn_shortString_l_04_capsule, center=<1.8175 0 0>, rotation=<rot 90 -179.9604 -179.9604>}]
 dyn_shortString_r_01 [{shapeType=Capsule, length=2, radius=1.5, shapeName=dyn_shortString_r_01_capsule, center=<1.5373 0 0>, rotation=<rot -90 179.9802 -179.9802>}]
 dyn_shortString_r_02 [{shapeType=Capsule, length=2, radius=1.5, shapeName=dyn_shortString_r_02_capsule, center=<1.8175 0 0>, rotation=<rot 90 -160.5288 -160.5288>}]
 dyn_shortString_r_03 [{shapeType=Capsule, length=3, radius=1.5, shapeName=dyn_shortString_r_03_capsule, center=<1.8175 0 0>, rotation=<rot 90 180 180>}]
 dyn_shortString_r_04 [{shapeType=Capsule, length=3, radius=1.5, shapeName=dyn_shortString_r_04_capsule, center=<1.8175 0 0>, rotation=<rot 90 179.9703 179.9703>}]
 pelvis [{shapeType=Capsule, length=8, radius=16, center=<-10.7507 -5.1454 0>, rotation=<rot -90 -179.8418 179.8417>}]
 thigh_l [{shapeType=Capsule, length=14.5, radius=13, center=<-16.3048 -1.4885 0.5848>, rotation=<rot 90 160.5288 160.5291>}]
 thigh_r [{shapeType=Capsule, length=14.5, radius=13, center=<16.3048 1.8837 -0.5848>, rotation=<rot -90 -179.9975 179.9974>}]
}
constraints [
 {bone1Name=thigh_l, bone2Name=pelvis, swing1LimitDegrees=45, swing2LimitDegrees=45, twistLimitDegrees=45}
 {bone1Name=calf_l, bone2Name=thigh_l, swing1LimitDegrees=45, swing2LimitDegrees=45, twistLimitDegrees=45}
 {bone1Name=thigh_r, bone2Name=pelvis, swing1LimitDegrees=45, swing2LimitDegrees=45, twistLimitDegrees=45}
 {bone1Name=calf_r, bone2Name=thigh_r, swing1LimitDegrees=45, swing2LimitDegrees=45, twistLimitDegrees=45}
 {bone1Name=dyn_shortString_l_01, bone2Name=pelvis, swing1LimitDegrees=45, swing2LimitDegrees=45, twistLimitDegrees=0}
 {bone1Name=dyn_shortString_l_02, bone2Name=dyn_shortString_l_01, swing1LimitDegrees=45, swing2LimitDegrees=30, twistLimitDegrees=0}
 {bone1Name=dyn_shortString_l_03, bone2Name=dyn_shortString_l_02, swing1LimitDegrees=45, swing2LimitDegrees=45, twistLimitDegrees=0}
 {bone1Name=dyn_shortString_l_04, bone2Name=dyn_shortString_l_03, swing1LimitDegrees=45, swing2LimitDegrees=45, twistLimitDegrees=0}
 {bone1Name=dyn_shortString_r_01, bone2Name=pelvis, swing1LimitDegrees=45, swing2LimitDegrees=45, twistLimitDegrees=0}
 {bone1Name=dyn_shortString_r_02, bone2Name=dyn_shortString_r_01, swing1LimitDegrees=45, swing2LimitDegrees=30, twistLimitDegrees=0}
 {bone1Name=dyn_shortString_r_03, bone2Name=dyn_shortString_r_02, swing1LimitDegrees=45, swing2LimitDegrees=45, twistLimitDegrees=0}
 {bone1Name=dyn_shortString_r_04, bone2Name=dyn_shortString_r_03, swing1LimitDegrees=45, swing2LimitDegrees=45, twistLimitDegrees=0}
]
