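# Physics asset named as the cloth colliders for the OatmealSpreadActive medium male player cosmetic body.
# It covers only the legs, with one body on each thigh and calf bone.
# Each calf is constrained to its same-side thigh with swing and twist left free, so the listed angle limits do not restrict rotation.
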PhysicsAsset /BRCosmetics/Characters/Player/Male/Medium/Bodies/M_MED_OatmealSpreadActive/Meshes/M_MED_OatmealSpreadActive_ClothColliders
bodies {calf_l [], calf_r [], thigh_l [], thigh_r []}
constraints [
 {
  bone1Name=calf_l
  bone2Name=thigh_l
  swing1LimitDegrees=45
  swing1Motion=Free
  swing2LimitDegrees=45
  swing2Motion=Free
  twistLimitDegrees=45
  twistMotion=Free
 }
 {
  bone1Name=calf_r
  bone2Name=thigh_r
  swing1LimitDegrees=45
  swing1Motion=Free
  swing2LimitDegrees=45
  swing2Motion=Free
  twistLimitDegrees=45
  twistMotion=Free
 }
]
